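# Translucent lit particle material for swirling desert dust on sprite and Niagara effects.
# A rotating flowmap or static normal map distorts two smoke textures, blended and eroded through dynamic parameters.
# Opacity combines particle alpha with depth, near-camera and camera-depth fades, keeping only the last on mobile.
# Color is tinted particle color, optionally shifted toward time-of-day shadow colors near intersecting geometry on non-mobile paths.
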Material /Game/Effects/Fort_Effects/Effects/Environment/Desert/Dust/Organize/M_SwirlDust
graph {
 properties {
  bUsedWithNiagaraSprites=true
  bUsedWithParticleSprites=true
  blendMode=BLEND_Translucent
  pixelDepthOffsetMode=PDOM_Legacy
  usedShadingModels=Default Lit
 }
 nodes [
  n1 = Rotator(Coordinate=TextureCoordinate()) {centerX=0.25, centerY=0.25, speed=0.5}
  n2 = TextureCoordinate()
  n3 = StaticSwitchParameter(
   True=ComponentMask(None=TextureSample(UVs=@n1) {samplerType=SAMPLERTYPE_LinearColor, texture=/Game/Global/Textures/Default/Flowmaps/T_SwirlFlow_02}.RGB) {g=true, r=true}
   False=ComponentMask(None=TextureSample(UVs=@n2) {samplerType=SAMPLERTYPE_Normal, texture=/Game/Global/Textures/Default/Flowmaps/T_SwirlFlow_02_N}.RGB) {g=true, r=true}
  ) {defaultValue=true, parameterName=UseVectorDisplacement}
  Multiply(
   A=@n3
   B=TextureSample(UVs=TextureCoordinate()) {samplerType=SAMPLERTYPE_LinearColor, texture=/Game/Global/Textures/Default/Masks_and_Gradients/RevealMask}.G
  )
  Rotator(Coordinate=@n2) {centerX=0.25, centerY=0.25, speed=0.5}
  n4 = Add(
   A=TextureCoordinate()
   B=Multiply(A=@n3, B=DynamicParameter() {defaultValue=<rgba 1 0.06 1 1>, paramNames [Erode,UVdistort,Softness,Param4]}.UVdistort)
  )
  n5 = TextureSample(UVs=@n4) {
   samplerSource=SSM_Wrap_WorldGroupSettings
   samplerType=SAMPLERTYPE_LinearColor
   texture=/Game/Effects/Fort_Effects/Textures/Smoke/T_SmokeSwirl
  }
  n6 = TextureSample() {
   samplerSource=SSM_Wrap_WorldGroupSettings
   samplerType=SAMPLERTYPE_LinearColor
   texture=&T_SmokeSwirl
  }
  n7 = Multiply(A=@n5.G, B=@n6.G)
  n8 = TextureSample(UVs=@n4) {
   samplerSource=SSM_Wrap_WorldGroupSettings
   samplerType=SAMPLERTYPE_LinearColor
   texture=/Game/Effects/Fort_Effects/Textures/Smoke/SmokeTest2_Soft
  }
  n9 = TextureSample() {
   samplerSource=SSM_Wrap_WorldGroupSettings
   samplerType=SAMPLERTYPE_LinearColor
   texture=&SmokeTest2_Soft
  }
  n10 = Multiply(A=@n8.G, B=@n9.G)
  n11 = DynamicParameter() {defaultValue=<rgba 1 0.06 1 1>, paramNames [Erode,UVdistort,Softness,Param4]}
  n12 = Multiply(A=Power(Base=LinearInterpolate(A=@n7, B=@n10, Alpha=@n11.Softness), Exp=@n11.Erode), B=ParticleColor().A)
  n13 = ScalarParameter() {defaultValue=500, parameterName=ShadowDistance}
  n14 = DepthFade(Opacity=@n12, FadeDistance=Multiply(A=@n13) {constB=0.8})
  n15 = ScalarParameter() {defaultValue=180, parameterName=NearCameraFadeDistance}
  n16 = MaterialFunctionCall(Distance=LinearInterpolate(A=Multiply(A=@n15) {constB=0.5}, B=@n15, Alpha=@n12) {constA=180, constB=60}) {materialFunction=&Near_Camera_Fade}
  n17 = OneMinus(None=MaterialFunctionCall(
    Fade Length=ScalarParameter() {defaultValue=750, parameterName="Fade Length (S)"}
    Fade Offset=ScalarParameter() {defaultValue=3000, parameterName="Fade Offset (S)"}
   ) {materialFunction=&CameraDepthFade}.Result)
  ShadingPathSwitch(Default=Multiply(A=Multiply(A=@n14, B=@n16.Result), B=@n17), Mobile=Multiply(A=@n12, B=@n17)) {desc=Migrated from FeatureLevelSwitch}
  n18 = MaterialFunctionCall(
   Color_Morning=VectorParameter() {parameterName=MorningShadowColor, defaultValue=<rgba 0.5312 0.3984 0.4068 1>}.RGB
   Color_Day=VectorParameter() {parameterName=DayShadowColor, defaultValue=<rgba 0.4375 0.3719 0.4314 1>}.RGB
   Color_Evening=VectorParameter() {parameterName=EveningShadowColor, defaultValue=<rgba 0.3906 0.3049 0.293 1>}.RGB
   Color_Night=VectorParameter() {parameterName=NightShadowColor, defaultValue=<rgba 0.5127 0.5643 0.6562 1>}.RGB
  ) {materialFunction=&Color_Time_of_Day}
  n19 = Multiply(A=VectorParameter() {parameterName=BeseColor, defaultValue=<rgba 1 1 1 0>}.RGB, B=ParticleColor().RGB)
  n20 = ScalarParameter() {defaultValue=80, parameterName=ColorFadeHeight}
  n21 = Clamp(None=DepthFade(FadeDistance=LinearInterpolate(A=@n20, B=@n13, Alpha=@n12) {constA=100, constB=600}))
  StaticSwitchParameter(
   True=ShadingPathSwitch(Default=LinearInterpolate(A=@n18.Result, B=@n19, Alpha=@n21), Mobile=@n19) {desc=Migrated from FeatureLevelSwitch}
   False=@n19
  ) {parameterName=UseDepthForShadow}
 ]
}
