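# Translucent, unlit, two-sided material for desert wind and dust streaks on spline-shaped particle meshes.
# World position offset combines SplineThicken with two panning-mask waves along a Tangent vector, masked by vertex red, with optional world-position phasing.
# Turbulence blends time-of-day shadow color into bright Color1.
# Opacity multiplies two panning noises by a UV edge falloff, then applies switchable angle, depth, camera and distance fades, time-of-day scaling and particle alpha.
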Material /Game/Effects/Fort_Effects/Effects/Environment/Desert/Dust/Organize/M_WindSplines
graph {
 properties {
  bUsedWithMeshParticles=true
  bUsedWithStaticLighting=true
  blendMode=BLEND_Translucent
  pixelDepthOffsetMode=PDOM_Legacy
  shadingModel=MSM_Unlit
  twoSided=true
  usedShadingModels=Unlit
 }
 nodes [
  n1 = VertexColor()
  n2 = MaterialFunctionCall(
   WidthBase=ScalarParameter() {defaultValue=244.5707, parameterName=WidthBase}
   WidthTip=ScalarParameter() {defaultValue=393.4102, parameterName=WidthTip}
   UVs for texturing=TextureCoordinate() {coordinateIndex=1}
   UVs For NormalShading=TextureCoordinate()
  ) {materialFunction=&SplineThicken}
  n3 = TextureSample(UVs=Panner(Coordinate=TextureCoordinate() {vTiling=3}, Speed=ScalarParameter() {defaultValue=1, parameterName=MicroWaveSpeed}) {speedY=0.2}) {texture=/Game/Effects/Fort_Effects/Textures/Masks/T_SplineThickness_Mask}
  n4 = Multiply(A=LinearInterpolate(Alpha=Clamp(None=@n3.G)) {constA=1, constB=-1}, B=ScalarParameter() {defaultValue=150, parameterName=MicroOffsetScale})
  n5 = VectorParameter() {parameterName=Tangent, defaultValue=<rgba 0.2 0.2 0.2 0>}
  n6 = TextureCoordinate()
  n7 = WorldPosition()
  n8 = Add(A=@n6, B=Multiply(A=ComponentMask(None=@n7.XYZ) {g=true, r=true}) {constB=0.04})
  n9 = TextureSample(UVs=Panner(
    Coordinate=StaticSwitchParameter(True=@n8, False=@n6) {parameterName=UseWorldPos_WaveOffset}
    Speed=ScalarParameter() {defaultValue=0.5, parameterName=WaveSpeed}
   ) {speedY=1}) {texture=&T_SplineThickness_Mask}
  n10 = Multiply(A=LinearInterpolate(Alpha=Clamp(None=@n9.G)) {constA=1, constB=-1}, B=ScalarParameter() {defaultValue=300, parameterName=OffsetScale})
  n11 = Transform(None=Add(A=Multiply(A=@n4, B=@n5.RGB), B=Multiply(A=@n5.RGB, B=@n10)))
  LinearInterpolate(
   A=Multiply(A=OneMinus(None=@n1.R), B=Add(A=@n2.WorldPositionOffset, B=@n11))
   B=Constant()
   Alpha=ScalarParameter() {parameterName=LerpOutWPO, sliderMax=1}
  )
  n12 = ScalarParameter() {defaultValue=1, parameterName=UMultiplier1}
  n13 = ScalarParameter() {defaultValue=1, parameterName=VMultiplier1}
  n14 = TextureSample(UVs=Multiply(A=TextureCoordinate() {uTiling=0.4, vTiling=2}, B=AppendVector(A=@n12, B=@n13))) {samplerType=SAMPLERTYPE_LinearColor, texture=&/T_Turbulence_Seamless}
  n15 = LinearInterpolate(
   A=MaterialFunctionCall(
    Color_Morning=VectorParameter() {parameterName=MorningShadowColor, defaultValue=<rgba 0.5776 0.1779 0.0976 1>}.RGB
    Color_Day=VectorParameter() {parameterName=DayShadowColor, defaultValue=<rgba 0.7011 0.521 0.2789 1>}.RGB
    Color_Evening=VectorParameter() {parameterName=EveningShadowColor, defaultValue=<rgba 0.8879 0.484 0.3515 1>}.RGB
    Color_Night=VectorParameter() {parameterName=NightShadowColor, defaultValue=<rgba 0.1141 0.1494 0.3712 1>}.RGB
   ) {materialFunction=&Color_Time_of_Day}.Result
   B=VectorParameter() {parameterName=Color1, defaultValue=<rgba 6.96 7.4808 8 0>}.RGB
   Alpha=@n14.G
  )
  MaterialFunctionCall(Emissive=@n15) {materialFunction=&MF_FX_EmissiveAdjustments}
  n16 = ScalarParameter() {defaultValue=150, parameterName=CameraNearFade Length, sliderMax=300}
  n17 = ScalarParameter() {defaultValue=50, parameterName=CameraNearFade Offset, sliderMax=100}
  n18 = ScalarParameter() {defaultValue=3, parameterName=AngleFade}
  n19 = TextureCoordinate() {uTiling=0.34}
  n20 = AppendVector(A=ScalarParameter() {defaultValue=1, parameterName=UMultiplier2}, B=ScalarParameter() {defaultValue=1, parameterName=VMultiplier2})
  n21 = Multiply(A=Time(), B=ScalarParameter() {defaultValue=1, parameterName=MaskSpeed})
  n22 = ScalarParameter() {defaultValue=0.012, parameterName=SpeedX1}
  n23 = ScalarParameter() {defaultValue=0.13, parameterName=SpeedY1}
  n24 = TextureSample(UVs=Panner(Coordinate=Multiply(A=@n19, B=@n20), Time=@n21, Speed=AppendVector(A=@n22, B=@n23)) {speedX=0.012, speedY=0.13}) {samplerType=SAMPLERTYPE_LinearColor, texture=&T_Turbulence_Seamless}
  n25 = TextureCoordinate() {uTiling=0.6, vTiling=1.5}
  n26 = AppendVector(A=ScalarParameter() {defaultValue=1, parameterName=UMultiplier3}, B=ScalarParameter() {defaultValue=1, parameterName=VMultiplier3})
  n27 = ScalarParameter() {defaultValue=0.012, parameterName=SpeedX2}
  n28 = ScalarParameter() {defaultValue=0.06, parameterName=SpeedY2}
  n29 = TextureSample(UVs=Panner(Coordinate=Multiply(A=@n25, B=@n26), Time=@n21, Speed=AppendVector(A=@n27, B=@n28)) {speedX=0.012, speedY=0.06}) {samplerType=SAMPLERTYPE_LinearColor, texture=&T_Turbulence_Seamless}
  n30 = ComponentMask(None=TextureCoordinate()) {g=true}
  n31 = ComponentMask(None=TextureCoordinate()) {r=true}
  n32 = Multiply(A=Multiply(A=OneMinus(None=@n30), B=@n30), B=Multiply(A=OneMinus(None=@n31), B=@n31))
  n33 = ScalarParameter() {defaultValue=0.8, parameterName=EXP}
  n34 = Multiply(
   A=Power(Base=Multiply(A=@n24.G, B=@n29.G), Exp=ScalarParameter() {defaultValue=1, parameterName=MaskPow})
   B=Multiply(A=Power(Base=@n32, Exp=@n33), B=ScalarParameter() {defaultValue=2.2, parameterName=ScaleAlphaMask})
  )
  n35 = StaticSwitchParameter(
   True=Multiply(
    A=MaterialFunctionCall("FadeSensitivity (optional)"=@n18) {materialFunction=&FadeBasedOnViewAngle}.Result
    B=@n34
   )
   False=@n34
  ) {parameterName=UseFadeAngle}
  n36 = StaticSwitchParameter(True=DepthFade(Opacity=@n35, FadeDistance=ScalarParameter() {defaultValue=64, parameterName=FadeDistance}), False=@n35) {parameterName=UseDepthFade}
  n37 = StaticSwitchParameter(
   True=Multiply(
    A=MaterialFunctionCall(Fade Length=@n16, Fade Offset=@n17) {materialFunction=&CameraDepthFade}.Result
    B=@n36
   )
   False=@n36
  ) {parameterName=UseCameraNearFade}
  n38 = StaticSwitchParameter(True=Multiply(A=DistanceCullFade(), B=@n37), False=@n37) {parameterName=UseDistanceFade}
  n39 = ScalarParameter() {defaultValue=1, parameterName=Opacity_Morning}
  n40 = ScalarParameter() {defaultValue=1, parameterName=Opacity_Day}
  n41 = ScalarParameter() {defaultValue=1, parameterName=Opacity_Evening}
  n42 = ScalarParameter() {defaultValue=1, parameterName=Opacity_Night}
  Multiply(
   A=ParticleColor().A
   B=MaterialFunctionCall(
    Color_Morning=Multiply(A=@n38, B=@n39)
    Color_Day=Multiply(A=@n38, B=@n40)
    Color_Evening=Multiply(A=@n38, B=@n41)
    Color_Night=Multiply(A=@n38, B=@n42)
   ) {materialFunction=&Color_Time_of_Day}.Result
  )
  Multiply(A=@n3.G, B=ComponentMask(None=TextureCoordinate()) {g=true})
  Multiply(A=@n9.G, B=ComponentMask(None=TextureCoordinate()) {g=true})
  Multiply(A=VertexColor().R, B=@n15)
  OneMinus(None=VertexColor().R)
 ]
}
